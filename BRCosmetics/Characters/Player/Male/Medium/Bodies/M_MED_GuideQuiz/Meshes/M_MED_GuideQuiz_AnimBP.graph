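# Post-process-style AnimBlueprint for the GuideQuiz male medium body cosmetic, adding procedural deformation over the pose copied from its parent mesh.
# A Control Rig takes flags for hidden arms, narrow back and a GuideQuiz glider, then weighted constraints blend belt deform bones between pelvis, spine, thigh-twist and groin bones.
# Each update sets those flags from both arms hidden, emote, skydive or parachute state, and an opened glider's GuideQuiz metatag.
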AnimBlueprint /BRCosmetics/Characters/Player/Male/Medium/Bodies/M_MED_GuideQuiz/Meshes/M_MED_GuideQuiz_AnimBP
target_skeleton=&/M_MED_GuideQuiz_Skeleton
graphs [
 AnimGraph {nodes [
   Output Pose(Result=Use cached pose 'Post Pose'()) {layerGroup=None}
   n1 = Control Rig!(
    Source=Copy Pose From Mesh() {bUseAttachedParent=true}
    Alpha
    ArmsHidden:bool=Get Is Arm Hidden() -> (Is Arm Hidden:bool)
    NarrowBack:bool=Get isNarrowBack() -> (isNarrowBack:bool)
    Is GuideQuiz Glider:bool=Get Is GuideQuiz Glider() -> (Is GuideQuiz Glider:bool)
   ) -> (Pose) {controlRigAssetReference {blueprintRigClass=&/M_MED_GuideQuiz_CtrlRig.M_MED_GuideQuiz_CtrlRig_C}}
   Save cached pose 'Post Deformation Rig'(Pose=@n1) {cacheName=Post Deformation Rig}
   n2 = "Constraint - deform_belt_03_bk_r (PPPP)"!(
    ComponentPose=Local To Component(LocalPose=Use cached pose 'Post Deformation Rig'())
    Alpha=1
    ConstraintWeights_0=0.7623
    ConstraintWeights_1:float=0.0906
    ConstraintWeights_2:float=0.0872
    ConstraintWeights_3:float=0.0599
   ) -> (Pose) {
    boneToModify {boneName=deform_belt_03_bk_r}
    constraintSetup [{targetBone {boneName=pelvis}}, {targetBone {boneName=spine_01}}, {targetBone {boneName=spine_02}}, {targetBone {boneName=spine_03}}]
    constraintWeights [1,1,1,1]
   }
   n3 = "Constraint - deform_belt_02_bk_r (PPPP)"!(
    ComponentPose=@n2
    Alpha=1
    ConstraintWeights_0=0.7605
    ConstraintWeights_1:float=0.0858
    ConstraintWeights_2:float=0.0946
    ConstraintWeights_3:float=0.0592
   ) -> (Pose) {
    boneToModify {boneName=deform_belt_02_bk_r}
    constraintSetup [{targetBone {boneName=pelvis}}, {targetBone {boneName=spine_01}}, {targetBone {boneName=spine_02}}, {targetBone {boneName=spine_03}}]
    constraintWeights [1,1,1,1]
   }
   n4 = "Constraint - deform_belt_01_bk_r (PPPP)"!(
    ComponentPose=@n3
    Alpha=1
    ConstraintWeights_0=0.7609
    ConstraintWeights_1:float=0.0845
    ConstraintWeights_2:float=0.0982
    ConstraintWeights_3:float=0.0564
   ) -> (Pose) {
    boneToModify {boneName=deform_belt_01_bk_r}
    constraintSetup [{targetBone {boneName=pelvis}}, {targetBone {boneName=spine_01}}, {targetBone {boneName=spine_02}}, {targetBone {boneName=spine_03}}]
    constraintWeights [1,1,1,1]
   }
   n5 = "Constraint - deform_belt_02_r (PPPPP)"!(
    ComponentPose=@n4
    Alpha=1
    ConstraintWeights_0=0.7554
    ConstraintWeights_1:float=0.0851
    ConstraintWeights_2:float=0.1001
    ConstraintWeights_3:float=0.0559
    ConstraintWeights_4:float=0.0036
   ) -> (Pose) {
    boneToModify {boneName=deform_belt_02_r}
    constraintSetup [
     {targetBone {boneName=pelvis}}
     {targetBone {boneName=spine_01}}
     {targetBone {boneName=spine_02}}
     {targetBone {boneName=spine_03}}
     {targetBone {boneName=thigh_twist_01_r}}
    ]
    constraintWeights [1,1,1,1,1]
   }
   n6 = "Constraint - deform_belt_01_r (PPPPPP)"!(
    ComponentPose=@n5
    Alpha=1
    ConstraintWeights_0=0.721
    ConstraintWeights_1:float=0.0964
    ConstraintWeights_2:float=0.1126
    ConstraintWeights_3:float=0.0635
    ConstraintWeights_4:float=0.0041
    ConstraintWeights_5:float=0.0024
   ) -> (Pose) {
    boneToModify {boneName=deform_belt_01_r}
    constraintSetup [
     {targetBone {boneName=pelvis}}
     {targetBone {boneName=spine_01}}
     {targetBone {boneName=spine_02}}
     {targetBone {boneName=spine_03}}
     {targetBone {boneName=thigh_twist_01_r}}
     {targetBone {boneName=deform_groin_r}}
    ]
    constraintWeights [1,1,1,1,1,1]
   }
   n7 = "Constraint - deform_belt_02_fr_r (PPPPP)"!(
    ComponentPose=@n6
    Alpha=1
    ConstraintWeights_0=0.6655
    ConstraintWeights_1:float=0.1178
    ConstraintWeights_2:float=0.1369
    ConstraintWeights_3:float=0.0797
    ConstraintWeights_4:float=0
   ) -> (Pose) {
    boneToModify {boneName=deform_belt_02_fr_r}
    constraintSetup [
     {targetBone {boneName=pelvis}}
     {targetBone {boneName=spine_01}}
     {targetBone {boneName=spine_02}}
     {targetBone {boneName=spine_03}}
     {targetBone {boneName=spine_04}}
    ]
    constraintWeights [1,1,1,1,1]
   }
   n8 = "Constraint - deform_belt_01_fr_r (PPPPP)"!(
    ComponentPose=@n7
    Alpha=1
    ConstraintWeights_0=0.6316
    ConstraintWeights_1:float=0.1277
    ConstraintWeights_2:float=0.1496
    ConstraintWeights_3:float=0.0887
    ConstraintWeights_4:float=0.0025
   ) -> (Pose) {
    boneToModify {boneName=deform_belt_01_fr_r}
    constraintSetup [
     {targetBone {boneName=pelvis}}
     {targetBone {boneName=spine_01}}
     {targetBone {boneName=spine_02}}
     {targetBone {boneName=spine_03}}
     {targetBone {boneName=spine_04}}
    ]
    constraintWeights [1,1,1,1,1]
   }
   n9 = "Constraint - deform_belt_front (PPPPP)"!(
    ComponentPose=@n8
    Alpha=1
    ConstraintWeights_0=0.647
    ConstraintWeights_1:float=0.1223
    ConstraintWeights_2:float=0.1433
    ConstraintWeights_3:float=0.0857
    ConstraintWeights_4:float=0.0017
   ) -> (Pose) {
    boneToModify {boneName=deform_belt_front}
    constraintSetup [
     {targetBone {boneName=pelvis}}
     {targetBone {boneName=spine_01}}
     {targetBone {boneName=spine_02}}
     {targetBone {boneName=spine_03}}
     {targetBone {boneName=spine_04}}
    ]
    constraintWeights [1,1,1,1,1]
   }
   n10 = "Constraint - deform_belt_01_fr_l (PPPPP)"!(
    ComponentPose=@n9
    Alpha=1
    ConstraintWeights_0=0.6316
    ConstraintWeights_1:float=0.1277
    ConstraintWeights_2:float=0.1496
    ConstraintWeights_3:float=0.0887
    ConstraintWeights_4:float=0.0025
   ) -> (Pose) {
    boneToModify {boneName=deform_belt_01_fr_l}
    constraintSetup [
     {targetBone {boneName=pelvis}}
     {targetBone {boneName=spine_01}}
     {targetBone {boneName=spine_02}}
     {targetBone {boneName=spine_03}}
     {targetBone {boneName=spine_04}}
    ]
    constraintWeights [1,1,1,1,1]
   }
   n11 = "Constraint - deform_belt_02_fr_l (PPPPP)"!(
    ComponentPose=@n10
    Alpha=1
    ConstraintWeights_0=0.6655
    ConstraintWeights_1:float=0.1178
    ConstraintWeights_2:float=0.1369
    ConstraintWeights_3:float=0.0797
    ConstraintWeights_4:float=0
   ) -> (Pose) {
    boneToModify {boneName=deform_belt_02_fr_l}
    constraintSetup [
     {targetBone {boneName=pelvis}}
     {targetBone {boneName=spine_01}}
     {targetBone {boneName=spine_02}}
     {targetBone {boneName=spine_03}}
     {targetBone {boneName=spine_04}}
    ]
    constraintWeights [1,1,1,1,1]
   }
   n12 = "Constraint - deform_belt_01_l (PPPPPP)"!(
    ComponentPose=@n11
    Alpha=1
    ConstraintWeights_0=0.721
    ConstraintWeights_1:float=0.0964
    ConstraintWeights_2:float=0.1126
    ConstraintWeights_3:float=0.0635
    ConstraintWeights_4:float=0.0041
    ConstraintWeights_5:float=0.0024
   ) -> (Pose) {
    boneToModify {boneName=deform_belt_01_l}
    constraintSetup [
     {targetBone {boneName=pelvis}}
     {targetBone {boneName=spine_01}}
     {targetBone {boneName=spine_02}}
     {targetBone {boneName=spine_03}}
     {targetBone {boneName=thigh_twist_01_l}}
     {targetBone {boneName=deform_groin_l}}
    ]
    constraintWeights [1,1,1,1,1,1]
   }
   n13 = "Constraint - deform_belt_02_l (PPPPP)"!(
    ComponentPose=@n12
    Alpha=1
    ConstraintWeights_0=0.7554
    ConstraintWeights_1:float=0.0851
    ConstraintWeights_2:float=0.1001
    ConstraintWeights_3:float=0.0559
    ConstraintWeights_4:float=0.0036
   ) -> (Pose) {
    boneToModify {boneName=deform_belt_02_l}
    constraintSetup [
     {targetBone {boneName=pelvis}}
     {targetBone {boneName=spine_01}}
     {targetBone {boneName=spine_02}}
     {targetBone {boneName=spine_03}}
     {targetBone {boneName=thigh_twist_01_l}}
    ]
    constraintWeights [1,1,1,1,1]
   }
   n14 = "Constraint - deform_belt_01_bk_l (PPPP)"!(
    ComponentPose=@n13
    Alpha=1
    ConstraintWeights_0=0.7609
    ConstraintWeights_1:float=0.0845
    ConstraintWeights_2:float=0.0982
    ConstraintWeights_3:float=0.0564
   ) -> (Pose) {
    boneToModify {boneName=deform_belt_01_bk_l}
    constraintSetup [{targetBone {boneName=pelvis}}, {targetBone {boneName=spine_01}}, {targetBone {boneName=spine_02}}, {targetBone {boneName=spine_03}}]
    constraintWeights [1,1,1,1]
   }
   n15 = "Constraint - deform_belt_02_bk_l (PPPP)"!(
    ComponentPose=@n14
    Alpha=1
    ConstraintWeights_0=0.7605
    ConstraintWeights_1:float=0.0858
    ConstraintWeights_2:float=0.0946
    ConstraintWeights_3:float=0.0592
   ) -> (Pose) {
    boneToModify {boneName=deform_belt_02_bk_l}
    constraintSetup [{targetBone {boneName=pelvis}}, {targetBone {boneName=spine_01}}, {targetBone {boneName=spine_02}}, {targetBone {boneName=spine_03}}]
    constraintWeights [1,1,1,1]
   }
   n16 = "Constraint - deform_belt_03_bk_l (PPPP)"!(
    ComponentPose=@n15
    Alpha=1
    ConstraintWeights_0=0.7623
    ConstraintWeights_1:float=0.0906
    ConstraintWeights_2:float=0.0872
    ConstraintWeights_3:float=0.0599
   ) -> (Pose) {
    boneToModify {boneName=deform_belt_03_bk_l}
    constraintSetup [{targetBone {boneName=pelvis}}, {targetBone {boneName=spine_01}}, {targetBone {boneName=spine_02}}, {targetBone {boneName=spine_03}}]
    constraintWeights [1,1,1,1]
   }
   Save cached pose 'Post Pose'(Pose=Component To Local(ComponentPose=@n16)) {cacheName=Post Pose}
  ]}
 EventGraph {nodes [
   Event BlueprintUpdateAnimation()
   > n1 = Sequence!(execute) -> (then_0, then_1, then_2:Exec)
   > Set Is Arm Hidden(Is Arm Hidden:bool=AND Boolean(A=Get bIsLeftArmHidden() -> (bIsLeftArmHidden:bool), B=Get bIsRightArmHidden() -> (bIsRightArmHidden:bool))) -> (Output_Get:bool)
   Set isNarrowBack(
    execute=@n1.then_1
    isNarrowBack:bool=OR Boolean!(A=Get bIsPlayingEmote(), B=Get bIsSkyDiving(), C:bool=Get bIsParachuteOpened()) -> (ReturnValue)
   ) -> (Output_Get:bool)
   Set Is GuideQuiz Glider(
    execute=@n1.then_2
    Is GuideQuiz Glider:bool=GliderItemDefinitionHasAnyMetaTag!(Tags:Array of Gameplay Tag Structures=Make Array("[0]"=(TagName="Cosmetics.Metatags.GuideQuiz")), bRequiresOpened:bool=true) -> (ReturnValue:bool)
   ) -> (Output_Get:bool)
  ]}
]
variables [
 Is Arm Hidden {type=boolean}
 Is GuideQuiz Glider {type=boolean}
 isNarrowBack {type=boolean}
]
